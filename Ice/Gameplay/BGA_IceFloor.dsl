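; BGA_IceFloor is a Fortnite ice floor actor that applies configured gameplay effects, apparently low friction, to pawns standing on it.
; Begin and end overlap events maintain a pawn set, which the authority reapplies effects to on a fast looping timer.
; Remote clients animate the mesh's scale up from zero with a timeline.
; On death, it activates a Niagara effect and hides the mesh.
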(event EventBeginPlay
  (Networking|SwitchHasAuthority
    (:Authority
      (Utilities|Time|SetTimerbyEvent (delegate Custom|Apply%20Low%20Friction) 0.1 true))
    (:Remote
      (Transformation|SetRelativeScale3D (Variables|Default|GetStaticMesh) "0, 0, 0")
      (bind (_direction _newtrack_0) (Timeline|Timeline_0 Play)
        (:Update
          (Transformation|SetRelativeScale3D (Variables|Default|GetStaticMesh) _newtrack_0))))))

(event Custom|Apply%20Low%20Friction
  (bind _result (Utilities|Set|ToArray (Variables|Default|GetOverlappedpawns)))
  (for _array_element _result
    (Fortnite|ApplyGameplayEffectsToActor _array_element (Variables|Default|GetEffects%20to%20Apply) (Transformation|GetActorLocation))))

(event Collision|EventActorBeginOverlap
  (params (OtherActor :type Actor%20Object%20Reference))
  (bind _asfort_pawn (Utilities|Casting|CastToFortPawn OtherActor))
  (Utilities|Set|Add (Variables|Default|GetOverlappedpawns) _asfort_pawn))

(event Collision|EventActorEndOverlap
  (params (OtherActor :type Actor%20Object%20Reference))
  (bind _asfort_pawn (Utilities|Casting|CastToFortPawn OtherActor))
  (Utilities|Set|Remove (Variables|Default|GetOverlappedpawns) _asfort_pawn))

(event EventOnDeathPlayEffects
  (params (Damage :type Float%20%28single-precision%29) (DamageTags :type Gameplay%20Tag%20Container%20Structure%20%28by%20ref%29) (Momentum :type Vector :default "0, 0, 0") (HitInfo :type Hit%20Result%20Structure%20%28by%20ref%29) (InstigatedBy :type Fort%20Pawn%20Object%20Reference) (DamageCauser :type Actor%20Object%20Reference) (EffectContext :type Gameplay%20Effect%20Context%20Handle%20Structure))
  (Components|Activation|Activate (Variables|Default|GetNiagara) false)
  (Rendering|SetVisibility (Variables|Default|GetStaticMesh) false false))

(fn ConstructionScript)
